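# Sound cue for the reskinned Brute monster's hard landing on dirt, built from eight large Brute footstep waves.
# A mixer layers two random picks of four waves each, with the second layer starting after a short random delay so the impact lands as a staggered double hit.
# A root modulator feeds the output, which uses Deimos step attenuation, global and local step concurrency, and the player/enemy movement sound class.
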SoundCue /Game/Sounds/Fort_Monster_Deimos/SFX/S18_Reskin/Brute_Reskin_LandHard_Dirt_Cue
root=Modulator_0
nodes [
 {id=Delay_0, children [Random_4], properties {delayMax=0.075, delayMin=0.025}}
 {id=Mixer_1, children [Random_3,Delay_0], properties {inputVolume [1,1]}}
 {id=Modulator_0, children [Mixer_1]}
 {
  id=Random_3
  children [WavePlayer_13, WavePlayer_14, WavePlayer_15, WavePlayer_16]
  properties {bShouldExcludeFromBranchCulling=true, weights [1,1,1,1]}
 }
 {
  id=Random_4
  children [WavePlayer_17, WavePlayer_18, WavePlayer_19, WavePlayer_20]
  properties {bShouldExcludeFromBranchCulling=true, weights [1,1,1,1]}
 }
 {id=WavePlayer_13, properties {soundWaveAssetPtr=&/Sideways_Brute_Footstep_Large_01}}
 {id=WavePlayer_14, properties {soundWaveAssetPtr=&/Sideways_Brute_Footstep_Large_02}}
 {id=WavePlayer_15, properties {soundWaveAssetPtr=&/Sideways_Brute_Footstep_Large_03}}
 {id=WavePlayer_16, properties {soundWaveAssetPtr=&/Sideways_Brute_Footstep_Large_04}}
 {id=WavePlayer_17, properties {soundWaveAssetPtr=&/Sideways_Brute_Footstep_Large_05}}
 {id=WavePlayer_18, properties {soundWaveAssetPtr=&/Sideways_Brute_Footstep_Large_06}}
 {id=WavePlayer_19, properties {soundWaveAssetPtr=&/Sideways_Brute_Footstep_Large_07}}
 {id=WavePlayer_20, properties {soundWaveAssetPtr=&/Sideways_Brute_Footstep_Large_08}}
]
properties {
 attenuationSettings=/Game/Sounds/Attenuation/Fort_Deimos_Step_Att
 maxDistance=2000
 priority=0
 soundClassObject=/Game/Sounds/FortSoundClassesAndModes/Fort_09_Player_Enemy_Movement
 virtualizationMode=Disabled
 volumeMultiplier=0.4
 attenuationOverrides {falloffDistance=1700, attenuationShapeExtents {x=300}}
 concurrencyOverrides {resolutionRule=PreventNew}
 concurrencySet [/Game/Sounds/Fort_Monster_Deimos/Settings/Deimos_Step_Global_Conc, /Game/Sounds/Fort_Monster_Deimos/Settings/Deimos_Step_Local_Conc]
}
