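# Sound cue for the melee swing of the reskinned Brute monster in the Deimos creature set.
# A mixer layers a random Deimos Fiend swing, pitched down and at half input volume, with a random Sideways Brute swing at full input volume.
# Playback uses Deimos melee attack attenuation, the creature sound class and global and local attack concurrency sets, with virtualization disabled.
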SoundCue /Game/Sounds/Fort_Monster_Deimos/SFX/S18_Reskin/Brute_Reskin_MeleeSwing_Cue
root=Mixer_0
nodes [
 {id=Mixer_0, children [Modulator_0,Modulator_4], properties {inputVolume [0.5,1]}}
 {id=Modulator_0, children [Random_3], properties {pitchMax=0.9, pitchMin=0.8}}
 {id=Modulator_4, children [Random_4], properties {pitchMax=1, pitchMin=1}}
 {
  id=Random_3
  children [WavePlayer_5,WavePlayer_6,WavePlayer_7,WavePlayer_8]
  properties {bShouldExcludeFromBranchCulling=true, weights [1,1,1,1]}
 }
 {
  id=Random_4
  children [WavePlayer_11, WavePlayer_15, WavePlayer_17, WavePlayer_18]
  properties {bShouldExcludeFromBranchCulling=true, weights [1,1,1,1]}
 }
 {id=WavePlayer_11, properties {soundWaveAssetPtr=&/Sideways_Brute_Swing_01}}
 {id=WavePlayer_15, properties {soundWaveAssetPtr=&/Sideways_Brute_Swing_02}}
 {id=WavePlayer_17, properties {soundWaveAssetPtr=&/Sideways_Brute_Swing_03}}
 {id=WavePlayer_18, properties {soundWaveAssetPtr=&/Sideways_Brute_Swing_04}}
 {id=WavePlayer_5, properties {soundWaveAssetPtr=/Game/Sounds/Fort_Monster_Deimos/SFX/Deimos_Fiend_Swing_01}}
 {id=WavePlayer_6, properties {soundWaveAssetPtr=/Game/Sounds/Fort_Monster_Deimos/SFX/Deimos_Fiend_Swing_02}}
 {id=WavePlayer_7, properties {soundWaveAssetPtr=/Game/Sounds/Fort_Monster_Deimos/SFX/Deimos_Fiend_Swing_03}}
 {id=WavePlayer_8, properties {soundWaveAssetPtr=/Game/Sounds/Fort_Monster_Deimos/SFX/Deimos_Fiend_Swing_04}}
]
properties {
 attenuationSettings=/Game/Sounds/Attenuation/Fort_Deimos_Attack_Melee_Att
 bDebug=true
 maxDistance=3300
 pitchMultiplier=0.98
 soundClassObject=/Game/Sounds/FortSoundClassesAndModes/Fort_Creatures
 virtualizationMode=Disabled
 volumeMultiplier=1.7835
 attenuationOverrides {distanceAlgorithm=Logarithmic}
 concurrencyOverrides {maxCount=3}
 concurrencySet [/Game/Sounds/Fort_Monster_Deimos/Settings/Deimos_Attack_Global_Conc, /Game/Sounds/Fort_Monster_Deimos/Settings/Deimos_Attack_Local_Conc]
}
