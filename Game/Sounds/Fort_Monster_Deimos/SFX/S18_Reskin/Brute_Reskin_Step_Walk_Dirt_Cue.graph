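# Sound cue for the walking footstep on dirt of a reskinned Deimos Brute monster, built from Sideways Brute large footstep samples.
# A Random node picks one of eight equally weighted wave players and is excluded from branch culling, feeding a Modulator that varies pitch.
# It uses Deimos step attenuation, global and local step concurrency sets and the player/enemy movement sound class, with lowest priority and virtualization disabled.
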SoundCue /Game/Sounds/Fort_Monster_Deimos/SFX/S18_Reskin/Brute_Reskin_Step_Walk_Dirt_Cue
root=Modulator_2
nodes [
 {id=Modulator_2, children [Random_1], properties {pitchMax=1.5, volumeMax=1}}
 {
  id=Random_1
  children [WavePlayer_9, WavePlayer_10, WavePlayer_11, WavePlayer_12, WavePlayer_13, WavePlayer_14, WavePlayer_15, WavePlayer_16]
  properties {bShouldExcludeFromBranchCulling=true, weights [1,1,1,1,1,1,1,1]}
 }
 {id=WavePlayer_10, properties {soundWaveAssetPtr=&/Sideways_Brute_Footstep_Large_02}}
 {id=WavePlayer_11, properties {soundWaveAssetPtr=&/Sideways_Brute_Footstep_Large_03}}
 {id=WavePlayer_12, properties {soundWaveAssetPtr=&/Sideways_Brute_Footstep_Large_04}}
 {id=WavePlayer_13, properties {soundWaveAssetPtr=&/Sideways_Brute_Footstep_Large_05}}
 {id=WavePlayer_14, properties {soundWaveAssetPtr=&/Sideways_Brute_Footstep_Large_06}}
 {id=WavePlayer_15, properties {soundWaveAssetPtr=&/Sideways_Brute_Footstep_Large_07}}
 {id=WavePlayer_16, properties {soundWaveAssetPtr=&/Sideways_Brute_Footstep_Large_08}}
 {id=WavePlayer_9, properties {soundWaveAssetPtr=&/Sideways_Brute_Footstep_Large_01}}
]
properties {
 attenuationSettings=/Game/Sounds/Attenuation/Fort_Deimos_Step_Att
 bBypassVolumeScaleForPriority=true
 bDebug=true
 maxDistance=2000
 priority=0
 soundClassObject=/Game/Sounds/FortSoundClassesAndModes/Fort_09_Player_Enemy_Movement
 virtualizationMode=Disabled
 volumeMultiplier=0.425
 concurrencyOverrides {maxCount=8, resolutionRule=StopOldest}
 concurrencySet [/Game/Sounds/Fort_Monster_Deimos/Settings/Deimos_Step_Global_Conc, /Game/Sounds/Fort_Monster_Deimos/Settings/Deimos_Step_Local_Conc]
}
