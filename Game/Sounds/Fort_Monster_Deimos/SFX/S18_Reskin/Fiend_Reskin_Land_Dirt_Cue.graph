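# Sound cue for the reskinned Fiend monster landing on dirt, reusing small Sideways Brute footstep recordings.
# A Random node picks one of four equally weighted variations and feeds it into a Modulator at the root.
# It plays at reduced volume in the player and enemy movement sound class with Deimos step attenuation.
# Global and local Deimos step concurrency sets limit simultaneous plays, and virtualization is disabled at lowest priority.
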SoundCue /Game/Sounds/Fort_Monster_Deimos/SFX/S18_Reskin/Fiend_Reskin_Land_Dirt_Cue
root=Modulator_3
nodes [
 {id=Modulator_3, children [Random_1]}
 {
  id=Random_1
  children [WavePlayer_0,WavePlayer_6,WavePlayer_7,WavePlayer_8]
  properties {bShouldExcludeFromBranchCulling=true, weights [1,1,1,1]}
 }
 {id=WavePlayer_0, properties {soundWaveAssetPtr=&/Sideways_Brute_Footstep_Small_01}}
 {id=WavePlayer_6, properties {soundWaveAssetPtr=&/Sideways_Brute_Footstep_Small_02}}
 {id=WavePlayer_7, properties {soundWaveAssetPtr=&/Sideways_Brute_Footstep_Small_03}}
 {id=WavePlayer_8, properties {soundWaveAssetPtr=&/Sideways_Brute_Footstep_Small_04}}
]
properties {
 attenuationSettings=/Game/Sounds/Attenuation/Fort_Deimos_Step_Att
 maxDistance=2000
 priority=0
 soundClassObject=/Game/Sounds/FortSoundClassesAndModes/Fort_09_Player_Enemy_Movement
 virtualizationMode=Disabled
 volumeMultiplier=0.4
 attenuationOverrides {falloffDistance=1700, attenuationShapeExtents {x=300}}
 concurrencyOverrides {resolutionRule=PreventNew}
 concurrencySet [/Game/Sounds/Fort_Monster_Deimos/Settings/Deimos_Step_Global_Conc, /Game/Sounds/Fort_Monster_Deimos/Settings/Deimos_Step_Local_Conc]
}
