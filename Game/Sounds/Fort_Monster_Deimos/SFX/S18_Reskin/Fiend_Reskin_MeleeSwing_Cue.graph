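# Sound cue for the melee swing of the Season 18 reskinned Deimos Fiend monster.
# A mixer layers a random pick of four original Fiend swings with a half-volume random pick of four Sideways Brute swings.
# Each layer has its own pitch modulator, with the Brute layer shifted higher.
# It plays in the creatures sound class with Deimos melee attenuation, global and local attack concurrency, and virtualization disabled.
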SoundCue /Game/Sounds/Fort_Monster_Deimos/SFX/S18_Reskin/Fiend_Reskin_MeleeSwing_Cue
root=Mixer_0
nodes [
 {id=Mixer_0, children [Modulator_1,Modulator_0], properties {inputVolume [1,0.5]}}
 {id=Modulator_0, children [Random_1], properties {pitchMax=1.25, pitchMin=1.1}}
 {id=Modulator_1, children [Random_0], properties {pitchMax=1.15}}
 {
  id=Random_0
  children [WavePlayer_0,WavePlayer_1,WavePlayer_2,WavePlayer_3]
  properties {bShouldExcludeFromBranchCulling=true, weights [1,1,1,1]}
 }
 {
  id=Random_1
  children [WavePlayer_5,WavePlayer_6,WavePlayer_7,WavePlayer_8]
  properties {bShouldExcludeFromBranchCulling=true, weights [1,1,1,1]}
 }
 {id=WavePlayer_0, properties {soundWaveAssetPtr=/Game/Sounds/Fort_Monster_Deimos/SFX/Deimos_Fiend_Swing_01}}
 {id=WavePlayer_1, properties {soundWaveAssetPtr=/Game/Sounds/Fort_Monster_Deimos/SFX/Deimos_Fiend_Swing_02}}
 {id=WavePlayer_2, properties {soundWaveAssetPtr=/Game/Sounds/Fort_Monster_Deimos/SFX/Deimos_Fiend_Swing_03}}
 {id=WavePlayer_3, properties {soundWaveAssetPtr=/Game/Sounds/Fort_Monster_Deimos/SFX/Deimos_Fiend_Swing_04}}
 {id=WavePlayer_5, properties {soundWaveAssetPtr=&/Sideways_Brute_Swing_01}}
 {id=WavePlayer_6, properties {soundWaveAssetPtr=&/Sideways_Brute_Swing_02}}
 {id=WavePlayer_7, properties {soundWaveAssetPtr=&/Sideways_Brute_Swing_03}}
 {id=WavePlayer_8, properties {soundWaveAssetPtr=&/Sideways_Brute_Swing_04}}
]
properties {
 attenuationSettings=/Game/Sounds/Attenuation/Fort_Deimos_Attack_Melee_Att
 bDebug=true
 maxDistance=3300
 pitchMultiplier=1.02
 soundClassObject=/Game/Sounds/FortSoundClassesAndModes/Fort_Creatures
 virtualizationMode=Disabled
 volumeMultiplier=1.902
 attenuationOverrides {distanceAlgorithm=Logarithmic}
 concurrencyOverrides {maxCount=3}
 concurrencySet [/Game/Sounds/Fort_Monster_Deimos/Settings/Deimos_Attack_Global_Conc, /Game/Sounds/Fort_Monster_Deimos/Settings/Deimos_Attack_Local_Conc]
}
