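# Sound cue for the reskinned Fiend monster's running footstep on dirt.
# A random node picks one of eight equally weighted small Sideways Brute footstep waves and feeds a modulator that varies pitch.
# It uses the Deimos step attenuation, the player/enemy movement sound class, and global and local Deimos step concurrency sets to limit overlapping steps, with virtualization disabled.
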SoundCue /Game/Sounds/Fort_Monster_Deimos/SFX/S18_Reskin/Fiend_Reskin_Step_Run_Dirt_Cue
root=Modulator_2
nodes [
 {id=Modulator_2, children [Random_2], properties {pitchMax=1.5, volumeMax=1}}
 {
  id=Random_2
  children [WavePlayer_8, WavePlayer_18, WavePlayer_19, WavePlayer_20, WavePlayer_21, WavePlayer_22, WavePlayer_23, WavePlayer_24]
  properties {bShouldExcludeFromBranchCulling=true, weights [1,1,1,1,1,1,1,1]}
 }
 {id=WavePlayer_18, properties {soundWaveAssetPtr=&/Sideways_Brute_Footstep_Small_02}}
 {id=WavePlayer_19, properties {soundWaveAssetPtr=&/Sideways_Brute_Footstep_Small_03}}
 {id=WavePlayer_20, properties {soundWaveAssetPtr=&/Sideways_Brute_Footstep_Small_04}}
 {id=WavePlayer_21, properties {soundWaveAssetPtr=&/Sideways_Brute_Footstep_Small_05}}
 {id=WavePlayer_22, properties {soundWaveAssetPtr=&/Sideways_Brute_Footstep_Small_06}}
 {id=WavePlayer_23, properties {soundWaveAssetPtr=&/Sideways_Brute_Footstep_Small_07}}
 {id=WavePlayer_24, properties {soundWaveAssetPtr=&/Sideways_Brute_Footstep_Small_08}}
 {id=WavePlayer_8, properties {soundWaveAssetPtr=&/Sideways_Brute_Footstep_Small_01}}
]
properties {
 attenuationSettings=/Game/Sounds/Attenuation/Fort_Deimos_Step_Att
 bBypassVolumeScaleForPriority=true
 bDebug=true
 maxDistance=2000
 priority=0
 soundClassObject=/Game/Sounds/FortSoundClassesAndModes/Fort_09_Player_Enemy_Movement
 virtualizationMode=Disabled
 volumeMultiplier=0.708
 attenuationOverrides {falloffDistance=1700, attenuationShapeExtents {x=300}}
 concurrencyOverrides {maxCount=3, resolutionRule=StopFarthestThenPreventNew}
 concurrencySet [/Game/Sounds/Fort_Monster_Deimos/Settings/Deimos_Step_Global_Conc, /Game/Sounds/Fort_Monster_Deimos/Settings/Deimos_Step_Local_Conc]
}
